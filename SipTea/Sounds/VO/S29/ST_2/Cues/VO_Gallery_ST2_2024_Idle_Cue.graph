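# Sound cue for idle voice lines in the SipTea Gallery ST2 2024 VO set under S29.
# A Random node picks one of eight idle wave variants with equal weight.
# Playback uses the shared close-range NPC VO attenuation, the non-critical enemy VO sound class and the SipTea chatter concurrency set.
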SoundCue /SipTea/Sounds/VO/S29/ST_2/Cues/VO_Gallery_ST2_2024_Idle_Cue
root=Random_0
nodes [
 {
  id=Random_0
  children [WavePlayer_0,WavePlayer_1,WavePlayer_2,WavePlayer_3,WavePlayer_4,WavePlayer_5,WavePlayer_6,WavePlayer_7]
  properties {weights [1,1,1,1,1,1,1,1]}
 }
 {id=WavePlayer_0, properties {soundWaveAssetPtr=/SipTea/Sounds/VO/S29/ST_2/Waves/VO_Gallery_ST2_2024_Idle_010}}
 {id=WavePlayer_1, properties {soundWaveAssetPtr=/SipTea/Sounds/VO/S29/ST_2/Waves/VO_Gallery_ST2_2024_Idle_020}}
 {id=WavePlayer_2, properties {soundWaveAssetPtr=/SipTea/Sounds/VO/S29/ST_2/Waves/VO_Gallery_ST2_2024_Idle_030}}
 {id=WavePlayer_3, properties {soundWaveAssetPtr=/SipTea/Sounds/VO/S29/ST_2/Waves/VO_Gallery_ST2_2024_Idle_040}}
 {id=WavePlayer_4, properties {soundWaveAssetPtr=/SipTea/Sounds/VO/S29/ST_2/Waves/VO_Gallery_ST2_2024_Idle_051}}
 {id=WavePlayer_5, properties {soundWaveAssetPtr=/SipTea/Sounds/VO/S29/ST_2/Waves/VO_Gallery_ST2_2024_Idle_061}}
 {id=WavePlayer_6, properties {soundWaveAssetPtr=/SipTea/Sounds/VO/S29/ST_2/Waves/VO_Gallery_ST2_2024_Idle_070}}
 {id=WavePlayer_7, properties {soundWaveAssetPtr=/SipTea/Sounds/VO/S29/ST_2/Waves/VO_Gallery_ST2_2024_Idle_080}}
]
properties {
 attenuationSettings=/Game/Sounds/Fort_Human_VO_BR/NPCs/Shared/NPC_VO_Close_Att
 maxDistance=8500
 soundClassObject=/Game/Sounds/FortSoundClassesAndModes/Fort_VO_Enemy_NonCritical
 volumeMultiplier=1
 concurrencySet [/SipTea/Sounds/VO/SipTea_Chatter_Conc]
}
